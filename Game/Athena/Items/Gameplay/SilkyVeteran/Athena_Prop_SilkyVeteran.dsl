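; Athena environmental slurp prop (SilkyVeteran) that grants pawns an effect while they overlap its SlurpNoCollision component.
; Once playlist data is ready, it destroys itself after a random delay if submerged or at or below ocean height, otherwise binds overlap events.
; Begin overlap applies the EnvSlurp gameplay effect unless an in-vehicle tag check matches.
; End overlap removes effects tagged as SilkyVeteran heal from the pawn.
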(event EventBeginPlay
  (bind (_returnvalue_x _returnvalue_y _returnvalue_z) (Math|Vector|BreakVector (Transformation|GetActorLocation)))
  (Class|FortAsyncActionCurrentPlaylistReady|PlaylistDataReadyAsync
    (:OnReady
      (if (or (Fortnite|WaterHelpers|IsActorSubmergedinOcean self) (<= _returnvalue_z (Fortnite|WaterHelpers|GetOceanTotalHeight)))
        (Utilities|FlowControl|Delay (Math|Random|RandomFloatinRange 1.0 1e+01))
        (Actor|DestroyActor)
        (else
          (Collision|AssignOnComponentBeginOverlap (Variables|Default|GetSlurpNoCollision) (delegate Custom|OnComponentBeginOverlap_Event_0))
          (Collision|AssignOnComponentEndOverlap (Variables|Default|GetSlurpNoCollision) (delegate Custom|OnComponentEndOverlap_Event_0)))))))

(event Custom|OnComponentEndOverlap_Event_0
  (params (OverlappedComponent :type Primitive%20Component%20Object%20Reference) (OtherActor :type Actor%20Object%20Reference) (OtherComp :type Primitive%20Component%20Object%20Reference) (OtherBodyIndex :type Integer))
  (bind _asfort_pawn (Utilities|Casting|CastToFortPawn OtherActor))
  (GameplayEffects|RemoveActiveEffectswithTags (Ability|GetAbilitySystemComponent _asfort_pawn) "(GameplayTags=((TagName=\"Asset.Silky.SilkyVeteran.Heal\")))"))

(event Custom|OnComponentBeginOverlap_Event_0
  (params (OverlappedComponent :type Primitive%20Component%20Object%20Reference) (OtherActor :type Actor%20Object%20Reference) (OtherComp :type Primitive%20Component%20Object%20Reference) (OtherBodyIndex :type Integer) (bFromSweep :type Boolean) (SweepResult :type Hit%20Result%20Structure%20%28by%20ref%29))
  (bind _asfort_pawn (Utilities|Casting|CastToFortPawn OtherActor))
  (if (not (GameplayTags|HasMatchingGameplayTag (Variables|Default|GetIn%20Vehicle%20Check)))
    (GameplayEffects|FortApplyGameplayEffectToTarget :self (Ability|GetFortAbilitySystemComponent self) :GameplayEffectClass "/Game/Athena/Items/Gameplay/SilkyVeteran/GE_Athena_EnvSlurp_Grant_SilkyVeteran.GE_Athena_EnvSlurp_Grant_SilkyVeteran_C" :TargetOfEffect (Ability|GetAbilitySystemComponent _asfort_pawn) :OptionalSourceObject self :OptionalEffectCauser self :Level 0.0)))

(fn ConstructionScript)
