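# Translucent, unlit, two-sided FX material for the SilkyVeteran gameplay item, named for the sides of the effect.
# A panning swirl normal map distorts the scaled UVs of a swirl noise mask, which pans on the same scaled time.
# A powered, boosted and saturated V-coordinate gradient multiplies that noise into a fading mask.
# A cyan ColorB parameter is defined, but no links from it or the mask to material outputs appear.
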Material /Game/Athena/Items/Gameplay/SilkyVeteran/FX/M_SilkyVeteran_Sides
graph {
 properties {
  blendMode=BLEND_Translucent
  pixelDepthOffsetMode=PDOM_Legacy
  shadingModel=MSM_Unlit
  translucencyLightingMode=TLM_VolumetricPerVertexDirectional
  twoSided=true
  usedShadingModels=Unlit
 }
 nodes [
  n1 = Power(Base=ComponentMask(None=TextureCoordinate()) {g=true}, Exp=ScalarParameter() {defaultValue=6})
  n2 = TextureCoordinate()
  n3 = ScalarParameter() {defaultValue=0.25, parameterName=TileSize}
  n4 = Multiply(A=Time(), B=ScalarParameter() {defaultValue=-1, parameterName=Panning Scale})
  n5 = TextureSampleParameter2D(UVs=Panner(Coordinate=Multiply(A=@n2, B=@n3), Time=@n4) {speedY=-0.028}) {
   desc=Distort the Tiling Mask by this
   parameterName=UVDistort_Normal
   samplerType=SAMPLERTYPE_Normal
   texture=/Game/Global/Textures/Default/Flowmaps/T_SwirlFlow_02_N
  }
  n6 = Multiply(A=ComponentMask(None=@n5.RGB) {g=true, r=true}, B=ScalarParameter() {defaultValue=0.075, parameterName=UVDistort})
  n7 = Multiply(
   A=TextureCoordinate()
   B=ComponentMask(None=VectorParameter() {parameterName=UV Scale, defaultValue=<rgba 1.6578 0.1 0 0>}.RGB) {g=true, r=true}
  )
  n8 = TextureSampleParameter2D(UVs=Panner(Coordinate=Add(A=@n6, B=@n7), Time=@n4) {speedY=-0.0125}) {
   desc=UV1 Tiling Noise Mask
   parameterName=TileAlphaMask
   samplerType=SAMPLERTYPE_LinearGrayscale
   texture=/Game/Global/Textures/Default/Noise/T_SwirlNoise_N_height
  }
  Multiply(A=Saturate(None=Multiply(A=@n1) {constB=10}), B=@n8.G)
  VectorParameter() {parameterName=ColorB, defaultValue=<rgba 0.0076 0.974 0.8173 1>}
 ]
}
