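# Unlit material for a Lumen plane in the SilkyVeteran gameplay item's FX.
# Its only node is a LumenColor vector parameter, white by default, with no connection to a material output shown.
# Automatic usage flagging in the editor is turned off.
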Material /Game/Athena/Items/Gameplay/SilkyVeteran/FX/M_SilyVeteran_LumenPlane
graph {
 properties {bAutomaticallySetUsageInEditor=false, shadingModel=MSM_Unlit, usedShadingModels=Unlit}
 nodes [VectorParameter() {parameterName=LumenColor, defaultValue=<rgba 1 1 1 1>}]
}
