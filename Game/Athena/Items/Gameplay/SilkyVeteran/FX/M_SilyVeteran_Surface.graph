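# Translucent, lit water-like surface material for the SilkyVeteran gameplay item's FX, with switches for Veteran or Dino edge masking.
# Panning water normals, with a separate Mobile/Forward variant, drive color blending (optionally time-of-day), metallic and roughness.
# Emissive combines depth-fade and UV-edge masks, flow-distorted noise and optional refracted caustics, omitted on Mobile.
# UV edge masks and particle alpha set opacity, with optional near-camera fade and noise-driven vertex displacement.
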Material /Game/Athena/Items/Gameplay/SilkyVeteran/FX/M_SilyVeteran_Surface
graph {
 properties {
  bUseMaterialAttributes=true
  bUsedWithNanite=true
  blendMode=BLEND_Translucent
  ditherOpacityMask=true
  pixelDepthOffsetMode=PDOM_Legacy
  previewMesh=/TheShape/Assets/SM_TheShape_A
  translucencyLightingMode=TLM_Surface
  usedShadingModels=Default Lit
 }
 nodes [
  n1 = ComponentMask(None=TextureCoordinate()) {g=true}
  n2 = ScalarParameter() {defaultValue=9, group=EdgeControl, parameterName=EdgeContrast}
  n3 = Saturate(None=Multiply(A=Power(Base=@n1, Exp=@n2)) {constB=10})
  n4 = ComponentMask(None=TextureCoordinate()) {r=true}
  n5 = Multiply(A=Power(Base=Sine(None=@n4) {period=2}, Exp=ScalarParameter() {group=EdgeControl, parameterName=EdgeControl_R_Dino})) {constB=10}
  n6 = StaticSwitchParameter(True=@n3, False=Saturate(None=Multiply(A=@n3, B=@n5))) {defaultValue=true, group=SwitchControls, parameterName=Use Veteran or Dino Edge Control}
  LinearInterpolate(A=LinearInterpolate(Alpha=ScalarParameter() {parameterName=SepTransCheck}) {constA=1.2}, Alpha=@n6) {constA=1.2}
  n7 = VectorParameter() {group=Color, parameterName=ColorA_Morning, defaultValue=<rgba 0.0008 0.1042 0.0874 1>}
  n8 = VectorParameter() {group=Color, parameterName="ColorA_Day(Default)", defaultValue=<rgba 0.0008 0.1042 0.0874 1>}
  n9 = VectorParameter() {group=Color, parameterName=ColorA_Evening, defaultValue=<rgba 0.0008 0.1042 0.0874 1>}
  n10 = VectorParameter() {group=Color, parameterName=ColorA_Night, defaultValue=<rgba 0.0008 0.1042 0.0874 1>}
  n11 = VectorParameter() {group=Color, parameterName=ColorB_Morning, defaultValue=<rgba 0.0076 0.974 0.8173 1>}
  n12 = VectorParameter() {group=Color, parameterName="ColorB_Day(Default)", defaultValue=<rgba 0.0076 0.974 0.8173 1>}
  n13 = VectorParameter() {group=Color, parameterName=ColorB_Evening, defaultValue=<rgba 0.0076 0.974 0.8173 1>}
  n14 = VectorParameter() {group=Color, parameterName=ColorB_Night, defaultValue=<rgba 0.0076 0.974 0.8173 1>}
  n15 = TextureCoordinate()
  n16 = Multiply(
   A=@n15
   B=ComponentMask(None=VectorParameter() {group=Normals, parameterName=NormalUVScale, defaultValue=<rgba 1 0.381 0 0>}.RGB) {g=true, r=true}
  ) {constB=0.8556}
  n17 = ScalarParameter() {defaultValue=0.2, parameterName=B_1}
  n18 = Multiply(A=Time(), B=ScalarParameter() {defaultValue=0.3, group=Normals, parameterName=Primary Panner Speed})
  n19 = TextureSample(UVs=Panner(Coordinate=Add(A=@n16, B=@n17) {constB=0.3215}, Time=@n18) {bFractionalPart=true, speedY=0.0417}) {samplerType=SAMPLERTYPE_Normal, texture=/Game/Global/Textures/Default/Tileable_Patterns/T_Water2_N}
  n20 = ComponentMask(None=@n19.RGB) {g=true, r=true}
  n21 = ScalarParameter() {defaultValue=0.1, group=Normals, parameterName=Noise 1}
  n22 = Multiply(
   A=@n15
   B=ComponentMask(None=VectorParameter() {group=Normals, parameterName=Normal2UVScale, defaultValue=<rgba 1 0.4667 0 0>}.RGB) {g=true, r=true}
  )
  n23 = TextureSample(UVs=Add(A=Multiply(A=@n20, B=@n21) {constB=0.04}, B=Panner(Coordinate=@n22, Time=@n18) {bFractionalPart=true, speedY=0.0375})) {samplerType=SAMPLERTYPE_Normal, texture=&T_Water2_N}
  n24 = ScalarParameter() {group=Normals, parameterName=NormalFlatten}
  n25 = LinearInterpolate(A=@n24, B=ScalarParameter() {defaultValue=0.99, group=Normals, parameterName=NormalFlatten_2}, Alpha=VertexColor().R)
  n26 = TextureCoordinate()
  n27 = VectorParameter() {group=Normals_Mobile, parameterName=NormalUVScale_M3, sortPriority=1, defaultValue=<rgba 1 0.4667 0 0>}
  n28 = ComponentMask(None=@n27.RGB) {g=true, r=true}
  n29 = Multiply(A=Time(), B=ScalarParameter() {defaultValue=0.3, group=Normals, parameterName=Mobile_Primary Panner Speed})
  n30 = VertexInterpolator(VS=Panner(Coordinate=Multiply(A=@n26, B=@n28), Time=@n29) {bFractionalPart=true, speedY=0.27})
  n31 = Multiply(
   A=@n26
   B=ComponentMask(None=VectorParameter() {group=Normals_Mobile, parameterName=NormalUVScale_M/M2, defaultValue=<rgba 1 0.381 0 0>}.RGB) {g=true, r=true}
  ) {constB=0.8556}
  n32 = TextureSample(UVs=VertexInterpolator(VS=Panner(Coordinate=@n31, Time=@n29) {bFractionalPart=true, speedY=0.33}).PS) {samplerType=SAMPLERTYPE_Normal, texture=&T_Water2_N}
  n33 = TextureSample(UVs=VertexInterpolator(VS=Panner(Coordinate=@n31, Time=@n29) {bFractionalPart=true, speedY=0.22}).PS) {samplerType=SAMPLERTYPE_Normal, texture=&T_Water2_N}
  n34 = ScalarParameter() {group=Normals, parameterName=NormalFlatten}
  n35 = LinearInterpolate(A=@n34, B=ScalarParameter() {defaultValue=0.99, group=Normals, parameterName=NormalFlatten_2}, Alpha=VertexColor().R)
  n36 = MaterialFunctionCall(
   BaseNormal=MaterialFunctionCall(
    Normal=TextureSample(UVs=@n30.PS) {samplerType=SAMPLERTYPE_Normal, texture=/Game/Global/Textures/Default/Tileable_Patterns/T_WaterSoft_N}.RGB
    Flatness=ScalarParameter() {defaultValue=0.9115, group=Normals_Mobile, parameterName=NormalFlaten_M3, sortPriority=5}
   ) {materialFunction=&FlattenNormal}.Result
   AdditionalNormal=MaterialFunctionCall(
    Normal=MaterialFunctionCall(BaseNormal=@n32.RGB, AdditionalNormal=@n33.RGB) {materialFunction=&BlendAngleCorrectedNormals}.Result
    Flatness=StaticSwitchParameter(True=@n35, False=@n34) {defaultValue=true, group=SwitchControls, parameterName=UseVertColorNormal}
   ) {materialFunction=&FlattenNormal}.Result
  ) {materialFunction=&BlendAngleCorrectedNormals}
  n37 = ShadingPathSwitch(
   Default=MaterialFunctionCall(
    Normal=MaterialFunctionCall(BaseNormal=@n19.RGB, AdditionalNormal=@n23.RGB) {materialFunction=&BlendAngleCorrectedNormals}.Result
    Flatness=StaticSwitchParameter(True=@n25, False=@n24) {defaultValue=true, group=SwitchControls, parameterName=UseVertColorNormal}
   ) {materialFunction=&FlattenNormal}.Result
   Forward=@n36.Result
   Mobile=@n36.Result
  )
  n38 = ComponentMask(None=@n37) {b=true}
  n39 = StaticSwitchParameter(
   True=LinearInterpolate(
    A=MaterialFunctionCall(Color_Morning=@n7.RGB, Color_Day=@n8.RGB, Color_Evening=@n9.RGB, Color_Night=@n10.RGB) {materialFunction=&Color_Time_of_Day}.Result
    B=MaterialFunctionCall(Color_Morning=@n11.RGB, Color_Day=@n12.RGB, Color_Evening=@n13.RGB, Color_Night=@n14.RGB) {materialFunction=&Color_Time_of_Day}.Result
    Alpha=@n38
   )
   False=LinearInterpolate(A=@n8.RGB, B=@n12.RGB, Alpha=@n38)
  ) {group=SwitchControls, parameterName=Use TOD Color}
  n40 = ScalarParameter() {defaultValue=1, parameterName=OverallBrightness_LOW}
  n41 = ScalarParameter() {defaultValue=1, parameterName=OverallBrightness}
  n42 = MaterialFunctionCall!(
   Low _ PC Low and Med=@n40
   "High _ XB1, PS4"=@n41
   "Epic _ PC High, XSX, PS5"=@n41
   --------
   "Mobile Low _ PC Perf, Android"=@n40
   Mobile Med _ Switch=@n40
  ) {materialFunction=/Game/Global/MaterialFunctions/TechArt/MF_Fortnite_QualitySwitch_Scalar}
  n43 = ScalarParameter() {defaultValue=34.9451, group=Emissive, parameterName=Fade Distance}
  n44 = Power(Base=ComponentMask(None=TextureCoordinate()) {g=true}, Exp=ScalarParameter() {defaultValue=9, group=EdgeControl, parameterName=EdgeContrast})
  n45 = Add(A=OneMinus(None=DepthFade(FadeDistance=@n43)), B=Saturate(None=Multiply(A=@n44) {constB=10}))
  n46 = TextureCoordinate()
  n47 = ScalarParameter() {defaultValue=0.25, group=Emissive, parameterName=TileSize}
  n48 = Multiply(A=Time(), B=ScalarParameter() {defaultValue=-1, group=Emissive, parameterName=Panning Scale})
  n49 = TextureSampleParameter2D(UVs=Panner(Coordinate=Multiply(A=@n46, B=@n47), Time=@n48) {speedY=-0.028}) {
   desc=Distort the Tiling Mask by this
   group=Emissive
   parameterName=UVDistort_Normal
   samplerType=SAMPLERTYPE_Normal
   texture=/Game/Global/Textures/Default/Flowmaps/T_SwirlFlow_02_N
  }
  n50 = Multiply(A=ComponentMask(None=@n49.RGB) {g=true, r=true}, B=ScalarParameter() {defaultValue=0.075, group=Emissive, parameterName=UVDistort})
  n51 = Multiply(
   A=TextureCoordinate()
   B=ComponentMask(None=VectorParameter() {group=Emissive, parameterName=UV Scale, defaultValue=<rgba 3 1 0 0>}.RGB) {g=true, r=true}
  )
  n52 = TextureSample(UVs=Panner(Coordinate=Add(A=@n50, B=@n51), Time=@n48) {speedY=-0.0125}) {samplerType=SAMPLERTYPE_LinearGrayscale, texture=/Game/Global/Textures/Default/Noise/T_SwirlNoise_N_height}
  n53 = Saturate(None=StaticSwitchParameter(True=Multiply(A=@n45, B=@n52.G), False=Constant()) {defaultValue=true, group=SwitchControls, parameterName=UseVeteran or Dino Emissive Mask})
  n54 = Constant()
  n55 = Multiply(A=@n37) {constA=1, desc=Normal}
  n56 = SceneDepth()
  n57 = Subtract(A=@n56.R, B=PixelDepth().R)
  n58 = ScalarParameter() {group=Caustics, parameterName=Min Refraction and Blur}
  n59 = Multiply(
   A=Divide(A=Max(A=@n57, B=@n58), B=@n56.R)
   B=ScalarParameter() {defaultValue=-1, desc=-0.4, group=Caustics, parameterName=Pixel Normal Offset}
  )
  n60 = Add(
   A=MaterialFunctionCall!(World Normal=Multiply(A=@n55, B=@n59), Custom UV Input) {materialFunction=/FortWater/Materials/Functions/Transform_Refraction_Normals}.View Space
   B=ScreenPosition().ViewportUV
  )
  n61 = MaterialFunctionCall(SceneDepth=SceneDepth(UVs=MaterialFunctionCall!(Screen Position=@n60) {desc=Dont pull in pixels from above the water, materialFunction=/FortWater/Materials/Functions/Distortion_MaskForeground}.Result).R) {materialFunction=&WorldPositionBehindTranslucency}
  n62 = ScalarParameter() {defaultValue=250, group=Caustics, parameterName=VatDepth}
  MaterialExpressionMultiply_9 = Multiply(A=CameraVectorWS()) {constB=-1}
  MaterialExpressionNormalize_0 = Normalize(VectorInput=Multiply(A=VectorParameter() {group=Caustics, parameterName=nynt, defaultValue=<rgba 0.5 0.5 0.25 0>}.RGB, B=@n37))
  MaterialExpressionScalarParameter_16 = ScalarParameter() {defaultValue=0.75, group=Caustics, parameterName=Relative Refractive Indices}
  n63 = Custom!(v=@MaterialExpressionMultiply_9, n=@MaterialExpressionNormalize_0, i=@MaterialExpressionScalarParameter_16) {
   code="return refract(v, n, i);"
   description=refract
   inputs [
    {inputName=v, input {expression=&self:MaterialExpressionMultiply_9}}
    {inputName=n, input {expression=&self:MaterialExpressionNormalize_0}}
    {inputName=i, input {expression=&self:MaterialExpressionScalarParameter_16}}
   ]
  }
  n64 = Multiply(
   A=Multiply(A=@n63, B=@n62) {desc=Refracted Vector * Depth}
   B=ScalarParameter() {defaultValue=-1, group=Caustics, parameterName=Caustic Refraction}
  )
  n65 = VectorParameter() {group=Caustics, parameterName=Vat Velocity}
  n66 = Multiply(
   A=MaterialFunctionCall!(
    WorldPosition=Add(A=@n61.Result, B=@n64)
    Scale=ScalarParameter() {defaultValue=512, group=Caustics, parameterName=Caustics Scale}
    Water Depth=@n62
    Time
    Water Velocity=ComponentMask(None=@n65.RGB) {g=true, r=true}
   ) {materialFunction=/FortWater/Materials/Functions/WaterCaustics}.Result
   B=VectorParameter() {group=Caustics, parameterName=Caustics, defaultValue=<rgba 0.0005 0.0677 0.0568 1>}.RGB
  )
  n67 = Add(
   A=Multiply(A=LinearInterpolate(B=@n39, Alpha=@n53), B=ScalarParameter() {defaultValue=0.5, group=Emissive, parameterName=Emissive Intensity})
   B=ShadingPathSwitch(Default=StaticSwitchParameter(True=@n54, False=@n66) {group=SwitchControls, parameterName=Remove Caustics}, Mobile=@n54) {desc=Migrated from FeatureLevelSwitch}
  )
  n68 = Multiply(A=VertexNormalWS(), B=ScalarParameter() {parameterName=WPO Amount})
  n69 = TextureCoordinate()
  n70 = ScalarParameter() {defaultValue=0.25, group=Emissive, parameterName=TileSize}
  n71 = Multiply(A=Time(), B=ScalarParameter() {defaultValue=-1, group=Emissive, parameterName=Panning Scale})
  n72 = TextureSampleParameter2D(UVs=Panner(Coordinate=Multiply(A=@n69, B=@n70), Time=@n71) {speedY=-0.028}) {
   desc=Distort the Tiling Mask by this
   group=Emissive
   parameterName=UVDistort_Normal
   samplerType=SAMPLERTYPE_Normal
   texture=&T_SwirlFlow_02_N
  }
  n73 = Multiply(A=ComponentMask(None=@n72.RGB) {g=true, r=true}, B=ScalarParameter() {defaultValue=0.075, group=Emissive, parameterName=UVDistort})
  n74 = Multiply(
   A=TextureCoordinate()
   B=ComponentMask(None=VectorParameter() {group=Emissive, parameterName=UV Scale, defaultValue=<rgba 3 1 0 0>}.RGB) {g=true, r=true}
  )
  n75 = TextureSample(UVs=Panner(Coordinate=Add(A=@n73, B=@n74), Time=@n71) {speedY=-0.0125}) {samplerType=SAMPLERTYPE_LinearGrayscale, texture=&T_SwirlNoise_N_height}
  n76 = ScalarParameter() {defaultValue=0.8, parameterName=Alpha}
  n77 = StaticSwitchParameter(True=LinearInterpolate(A=@n76, Alpha=@n6), False=Multiply(A=@n76, B=@n6)) {defaultValue=true, group=SwitchControls, parameterName=Use Veteran or Dino Edge Control}
  n78 = MaterialFunctionCall(Distance=ScalarParameter() {parameterName=Near Camera Fade Distance}) {materialFunction=&Near_Camera_Fade}
  n79 = Multiply(
   A=StaticSwitchParameter(True=Multiply(A=@n77, B=@n78.Result), False=@n77) {group=SwitchControls, parameterName=Use Near Camera Fade}
   B=ParticleColor().A
  )
  n80 = SetMaterialAttributes!(
   MaterialAttributes
   Base Color=Multiply(A=@n39, B=@n42.Result)
   Metallic=LinearInterpolate(B=ScalarParameter() {defaultValue=0.5, group=Color, parameterName=Metallic_Max}, Alpha=@n38) {constB=0.5}
   Roughness=LinearInterpolate(
    A=ScalarParameter() {defaultValue=0.1, group=Color, parameterName=Roughness_Max}
    B=ScalarParameter() {defaultValue=0.01, group=Color, parameterName=Roughness_Min}
    Alpha=@n38
   ) {constA=0.1, constB=0.01}
   Emissive Color=Multiply(A=Saturate(None=@n67), B=@n42.Result)
   Opacity=@n79
   Opacity Mask=@n79
   Normal=@n37
   World Position Offset=StaticSwitchParameter(True=Multiply(A=@n68, B=@n75.G), False=Constant()) {group=SwitchControls, parameterName=Use WPO}
   Specular=ScalarParameter() {defaultValue=0.5, parameterName=Spec}
  ) {attributeSetTypes [
    #1
    #2
    #3
    #4
    #5
    #6
    #7
    #8
    #9
   ]}
  MaterialFunctionCall(MA In=@n80) {materialFunction=&MF_CreativeMoveTool}
 ]
}
