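# Material function that outputs a spherical world-space mask around a Slurp Vat location, for SilkyVeteran item effects.
# A SphereMask compares world position with the VatLocation vector parameter, shaped by the VatRadius and SphereHardness scalars.
# All three parameters share the Slurp Vat group, so materials using the function expose the mask's center, size and edge falloff.
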MaterialFunction /Game/Athena/Items/Gameplay/SilkyVeteran/FX/SlurpVatMask
graph {nodes [FunctionOutput(None=SphereMask(
    A=WorldPosition().XYZ
    B=VectorParameter() {group=Slurp Vat, parameterName=VatLocation}.RGB
    Radius=ScalarParameter() {defaultValue=850, group=Slurp Vat, parameterName=VatRadius, sortPriority=1}
    Hardness=ScalarParameter() {group=Slurp Vat, parameterName=SphereHardness, sortPriority=2}
   ))]}
